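# Niagara system for the engine effect of the Moth glider cosmetic, apparently converted from an older effect, judging by its name.
# It combines three emitters: Bright_Core, Flame_Left and Light.
# Dynamic inputs include float and vector curves, a waveform, color scaling and a front-end float modifier.
# User parameters expose a skeletal mesh and front-end and front-end-preview flags; emitter internals and renderers are not detailed.
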NiagaraSystem /BRCosmetics/Gadgets/Assets/VinderTech_GliderChute/Glider_Moth/FX/NS_Converted_Glider_Moth_Engine
dependencies {
 usedDataInterfaces [&NiagaraDataInterfaceCurve, &NiagaraDataInterfaceVectorCurve]
 usedDynamicInputs [
  &FN_FrontEndFloatModifier
  &FloatFromCurve
  &VectorFromCurve
  &Multiply_LinearColorByFloat
  &Waveform
 ]
 usedModules [
  &SystemState
  &self:SystemSpawnScript.SystemScriptSource.SystemScriptGraph.NiagaraNodeAssignment_0.SetVariables_#1
 ]
}
emitters {Bright_Core {rendererClasses [], renderers []}, Flame_Left {rendererClasses [], renderers []}, Light {rendererClasses [], renderers []}}
summary {
 systemName=NS_Converted_Glider_Moth_Engine
 emitters [{emitterName=Light, rendererClasses []}, {emitterName=Flame_Left, rendererClasses []}, {emitterName=Bright_Core, rendererClasses []}]
 userVariables [
  User.bIsFrontEnd {type=NiagaraBool}
  User.bIsFrontEndPreview {type=NiagaraBool}
  User.SkeletalMesh {
   type=NiagaraDataInterfaceSkeletalMesh
   defaultValue {
    struct=NiagaraExt_VariableValue_DataInterface
    value {dataInterface=&self:NiagaraDataInterfaceSkeletalMesh_0}
   }
  }
 ]
}
